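# Niagara idle effect for the ShadowLap character cosmetic, emitting tintable "goop" sprites from the character's body.
# One emitter splits particles into seven partitions, each mapped to a socket with its own spawn shape.
# Lifetime shortens with owner speed, while inherited source movement fades into counter-force and particles later drift upward.
# Distance from the system shapes sprite size and culls strays, while dynamic material parameters and camera offset shape the look.
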NiagaraSystem /CosmeticCompanions/FX/Builtins/ShadowLap/Idle/NS_ShadowLap_Char_Idle
dependencies {
 usedDataInterfaces [&NiagaraDataInterfaceCurve, &NiagaraDataInterfaceSocketReader]
 usedDynamicInputs [&ReturnMeshOrientationAxis]
 usedModules [
  &self:SystemSpawnScript.SystemScriptSource.SystemScriptGraph.NiagaraNodeAssignment_3.SetVariables_#1
  &SystemState
 ]
}
emitters {Goop_All {
  emitterUpdateScript [
   &EmitterState()
   &SpawnRate(SpawnRate=30, Use Spawn Probability=false)
  ]
  particleSpawnScript [
   &PartitionParticles(Number Of Partitions=7)
   &InitializeParticle(
    Color=&MultiplyLinearColor(Linear Color=$User.EmissiveTint, Scale Factor=<1 1 1 3>)
    Lifetime=&Lerp_Float(
     A=&RandomRangeFloat(Minimum=1.4, Maximum=2.2)
     B=0.5
     Alpha=&Divide_Float(A=&VectorLength(Vector=$Engine.Owner.Velocity), B=5000)
    )
    Sprite Size Mode=::NewEnumerator1
    Uniform Sprite Size Max=7.5
    Uniform Sprite Size Min=2.5
    UsePositionOffset=false
    Write Material Random=true
   )
   &SocketLocation(
    Orientation Sampling=::NewEnumerator1
    SamplePosOffset=true
    Sampled Orientation Coordinate Space:ENiagaraCoordinateSpace=::World
    Sampled Position Offset=&MergeHelper_VectorSwitch(
     0=<0 0 0>
     1=<0 0 0>
     2=<0 0 0>
     3=<0 0 0>
     4=<0 0 0>
     5=<0 0 0>
     6=<-15 0 0>
     Index=$Particles.PartitionParticles.Partition
     Num=::NewEnumerator5
    )
    Socket Index=$Particles.PartitionParticles.Partition
    Socket Reader=$System.Socket Reader
    Socket Sampling Mode=::NewEnumerator3
    VelocitySpeedLimit=false
   )
   &ShapeLocation(
    Additional Quaternion Rotation=false
    Hemisphere Distribution=&MergeHelper_Vector2Switch(0=<1 1>, 1=<1 1>, 2=<1 1>, 3=<1 1>, 4=<1 1>, 5=<1 1>, 6=<0 1>, Index=$Particles.PartitionParticles.Partition, Num=::NewEnumerator5)
    MaskSpawnGroup=false
    Non Uniform Scale=&MergeHelper_VectorSwitch(
     0=<1 1 1>
     1=<1 1 2>
     2=<1 1 2>
     3=<1 1 2.5>
     4=<1 1 2.5>
     5=<1 1 1>
     6=<15 1 12.5>
     Index=$Particles.PartitionParticles.Partition
     Num=::NewEnumerator5
    )
    Override Seed=false
    Rotation Mode=::NewEnumerator3
    Rotation Quaternion=$Output.SocketLocation.SampledRotation
    Sphere Radius=&MergeHelper_FloatSwitch(0=7.5, 1=3.5, 2=3.5, 3=2, 4=2, 5=3.5, 6=1, Index=$Particles.PartitionParticles.Partition, Num=::NewEnumerator5)
    Sphere Surface Distribution:NiagaraFloat=&MergeHelper_FloatSwitch(0=0, 1=0, 2=0, 3=0, 4=0, 5=0, 6=1, Index=$Particles.PartitionParticles.Partition, Num=::NewEnumerator5)
    UseOwnerScale=true
   )
   &AddVelocity(Velocity=<0 0 1>, Velocity Speed Scale=&RandomRangeFloat(Minimum=0.1, Maximum=1))
  ]
  particleUpdateScript [
   &ParticleState()
   &ScaleSpriteSize(
    Scale Sprite Size Mode=::NewEnumerator0
    Uniform Scale Factor=&FloatFromCurve(
     FloatCurve=DataInterface() {propertyValues {Curve {keys [
         {arriveTangent=9.1187, interpMode=RCIM_Cubic, time=0, value=0}
         {arriveTangent=0.5833, interpMode=RCIM_Cubic, leaveTangent=0.5833, time=0.2487, value=1.1246}
         {arriveTangent=-1.497, interpMode=RCIM_Cubic, leaveTangent=-1.497, time=0.6, value=0.35}
         <at 1 0 RCIM_Cubic>
        ]}}}
     Scale Curve=&NormalizeDistanceRange(
      Start Position:NiagaraPosition=/Niagara/DynamicInputs/Helpers/SimulationPosition()
      End Position:NiagaraPosition=$Particles.Position
      Distance:NiagaraFloat=175
      Invert Normalized Range:NiagaraBool=true
     )
    )
   )
   &Drag(Drag=5, Use Linear Drag=true, Use Rotational Drag=true)
   &GravityForce(Gravity=&MakeVector(
     X=0
     Y=0
     Z=&FloatFromCurve(FloatCurve=DataInterface() {propertyValues {Curve {keys [<at 0.3 0>,<at 1 1>]}}}, Scale Curve=250)
    ))
   &DynamicMaterialParameters(
    Index 0 Param 1=5
    Index 0 Param 2=12
    Index 0 Param 3=&FloatFromCurve(FloatCurve=DataInterface() {propertyValues {Curve {keys [<at 0 5 RCIM_Cubic>, {arriveTangent=-4.1151, interpMode=RCIM_Cubic, leaveTangent=-4.1151, tangentMode=RCTM_User, time=1, value=1}]}}})
    Module.Param0WriteEnabled=true
    Module.Param1WriteEnabled=true
    Module.Param2WriteEnabled=true
    Module.Param3WriteEnabled=false
   )
   &InheritSourceMovement(
    Applied Force Scale=&VectorFromFloat(Value=&FloatFromCurve(FloatCurve=DataInterface() {propertyValues {Curve {keys [<at 0 0>,<at 1 -0.5>]}}}))
    Applied Position Scale=&VectorFromFloat(Value=&FloatFromCurve(FloatCurve=DataInterface() {propertyValues {Curve {
         defaultValue=0
         keys [
          {arriveTangent=-9.3774, interpMode=RCIM_Cubic, leaveTangent=-9.3774, tangentMode=RCTM_User, time=0, value=1}
          {interpMode=RCIM_Cubic, tangentMode=RCTM_User, time=0.25, value=0}
         ]
        }}}))
    ApplyForce=true
    ApplyForceFalloff=false
    ApplyPosition=true
    ApplyPositionFalloff=false
    Module.LimitForceSpeed=false
   ) {module=InheritSourceMovement001}
   &CameraOffset(Camera Offset Amount=&Multiply_Float(
     A=&FloatFromCurve(FloatCurve=DataInterface() {propertyValues {Curve {keys [{interpMode=RCIM_Cubic, tangentMode=RCTM_User, time=0.5, value=1}, {interpMode=RCIM_Cubic, tangentMode=RCTM_User, time=1, value=0}]}}})
     B=7.5
    ))
   &SolveForcesAndVelocity(Clamp Velocity=false, Limit Acceleration=false)
   &ScaleColor(ScaleA=true, ScaleRGB=true)
   &KillParticlesInVolume(Invert Volume=true, Sphere Radius=200)
  ]
  rendererClasses [&NiagaraSpriteRendererProperties]
  renderers [{rendererClass=&NiagaraSpriteRendererProperties}]
 }}
summary {
 systemName=NS_ShadowLap_Char_Idle
 emitters [{emitterName=Goop_All, rendererClasses [&NiagaraSpriteRendererProperties]}]
 userVariables [User.EmissiveTint {type=LinearColor, description="", defaultValue {struct=LinearColor, value=<rgba 1 1 1 1>}}]
}
